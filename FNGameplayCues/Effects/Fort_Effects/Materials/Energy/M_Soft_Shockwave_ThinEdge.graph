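# Translucent unlit material for mesh particles that draws a soft shockwave with a thin edge, colored by particle color through the shared emissive adjustments function.
# A contrasted Fresnel multiplied by its contrasted inverse isolates the edge band, with its exponent driven by the EdgeThickness dynamic parameter.
# The brightness-scaled band is masked by an inverted V-gradient falloff and particle alpha for opacity.
# Depth fade softens intersections, except on mobile, which skips it.
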Material /FNGameplayCues/Effects/Fort_Effects/Materials/Energy/M_Soft_Shockwave_ThinEdge
graph {
 properties {bUsedWithMeshParticles=true, blendMode=BLEND_Translucent, pixelDepthOffsetMode=PDOM_Legacy, shadingModel=MSM_Unlit, usedShadingModels=Unlit}
 nodes [
  n1 = ParticleColor()
  MaterialFunctionCall(Emissive=@n1.RGB) {materialFunction=&MF_FX_EmissiveAdjustments}
  n2 = OneMinus(None=MaterialFunctionCall() {materialFunction=&LinearGradient}.VGradient)
  n3 = ScalarParameter() {defaultValue=0.05, parameterName=BottomFalloff}
  n4 = DynamicParameter() {paramNames [EdgeThickness, Param2, Param3, Param4]}
  n5 = MaterialFunctionCall(In=Fresnel(ExponentIn=@n4.EdgeThickness) {exponent=1}, Contrast=Constant() {r=0.1}) {materialFunction=&CheapContrast}
  n6 = MaterialFunctionCall(In=OneMinus(None=Fresnel(ExponentIn=@n4.EdgeThickness) {exponent=1}), Contrast=Constant() {r=1}) {materialFunction=&CheapContrast}
  n7 = Clamp(None=Multiply(A=Multiply(A=@n5.Result, B=@n6.Result), B=ScalarParameter() {defaultValue=1, parameterName=FresnelBrightness}))
  n8 = DepthFade(Opacity=@n7)
  Multiply(
   A=@n1.A
   B=Multiply(
    A=MaterialFunctionCall(In=@n2, Contrast=@n3) {materialFunction=&CheapContrast}.Result
    B=ShadingPathSwitch(Default=@n8, Mobile=@n7) {desc=Migrated from FeatureLevelSwitch}
   )
  )
 ]
}
