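# Translucent, unlit, two-sided material for mesh particles, apparently a storm damage cylinder effect, tinted and faded by particle color and alpha.
# A panning swirl mask distorts scrolling gradient-section UVs, and a vertical gradient shapes the streaks and blends one end to solid.
# Inverted depth fade boosts glow at geometry intersections, while depth fade and inverted Fresnel soften opacity.
# The Color parameter is unconnected.
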Material /FNGameplayCues/Effects/Fort_Effects/Materials/Energy/M_Storm_Damage_Cylinder
graph {
 properties {
  bUsedWithMeshParticles=true
  blendMode=BLEND_Translucent
  pixelDepthOffsetMode=PDOM_Legacy
  shadingModel=MSM_Unlit
  twoSided=true
  usedShadingModels=Unlit
 }
 nodes [
  n1 = ParticleColor()
  n2 = Multiply(A=OneMinus(None=DepthFade() {fadeDistanceDefault=512}), B=@n1.RGB)
  MaterialFunctionCall(Emissive=Add(A=Multiply(A=@n2) {constB=10}, B=@n1.RGB)) {materialFunction=&MF_FX_EmissiveAdjustments}
  n3 = Panner(Coordinate=TextureCoordinate() {uTiling=8, vTiling=4}) {speedY=0.1}
  n4 = TextureCoordinate() {uTiling=6}
  n5 = Add(
   A=Multiply(A=TextureSample(UVs=@n3) {samplerType=SAMPLERTYPE_LinearGrayscale, texture=/Game/Global/Textures/Default/Flowmaps/T_StormSwirlMask}.G) {constB=0.015}
   B=Panner(Coordinate=MaterialFunctionCall(XY=@n4) {materialFunction=&Swizzle}.YX) {speedX=0.25, speedY=0.25}
  )
  n6 = MaterialFunctionCall() {materialFunction=&LinearGradient}
  n7 = Power(Base=@n6.VGradient) {constExponent=5}
  n8 = LinearInterpolate(
   A=Multiply(A=TextureSample(UVs=@n5) {texture=/Game/Global/Textures/Default/Masks_and_Gradients/T_Vert_Gradient_Sections}.G, B=Power(Base=@n6.VGradient))
   B=Constant() {r=1}
   Alpha=Clamp(None=Multiply(A=@n7) {constB=2})
  )
  n9 = Power(Base=OneMinus(None=Fresnel() {exponent=1}), Exp=ScalarParameter() {defaultValue=1.5, parameterName=AngleFadeAmount})
  Multiply(A=@n1.A, B=Multiply(A=DepthFade(Opacity=@n8) {fadeDistanceDefault=512}, B=Clamp(None=@n9)))
  VectorParameter() {parameterName=Color, defaultValue=<rgba 2 0.5262 0.2464 1>}
 ]
}
