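# Translucent, unlit, two-sided particle material for a storm energy wave.
# Three panned electric-noise samples, timed by the PannSpeed dynamic parameter, form a mask driving normal-based displacement.
# A second offset term scales vertices about the particle position by inverse vertex red and ScaleModifier.
# World-aligned grid texture and V gradients blend two HDR colors and shape opacity, faded by view angle, particle alpha and depth.
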Material /FNGameplayCues/Effects/Fort_Effects/Materials/Energy/M_Storm_Energy_Wave
graph {
 properties {
  bUsedWithParticleSprites=true
  blendMode=BLEND_Translucent
  pixelDepthOffsetMode=PDOM_Legacy
  shadingModel=MSM_Unlit
  twoSided=true
  usedShadingModels=Unlit
 }
 nodes [
  n1 = TextureCoordinate() {uTiling=6, vTiling=0.25}
  n2 = ScalarParameter() {defaultValue=1, parameterName=UVScale}
  n3 = Multiply(A=DynamicParameter() {defaultValue=<rgba 1 0 1 1>, paramNames [Param1, ScaleModifier, PannSpeed, DisplaceAmount]}.PannSpeed, B=Time())
  n4 = TextureSample(UVs=Panner(Coordinate=Multiply(A=@n2, B=@n1), Time=@n3) {speedX=0.05, speedY=1}) {samplerType=SAMPLERTYPE_LinearGrayscale, texture=/Game/Characters/Player/Male/Medium/Bodies/M_MED_SRDrift/Textures/T_ElectricNoise}
  n5 = TextureCoordinate() {uTiling=15, vTiling=0.5}
  n6 = TextureSample(UVs=Panner(Coordinate=Multiply(A=@n2, B=@n5), Time=@n3) {speedX=-0.025, speedY=1}) {samplerType=SAMPLERTYPE_LinearGrayscale, texture=&T_ElectricNoise}
  n7 = Panner(Coordinate=Multiply(A=@n2, B=TextureCoordinate() {uTiling=25, vTiling=2}), Time=@n3) {speedX=-0.05, speedY=4}
  n8 = Clamp(None=Add(
    A=Add(A=@n4.G, B=@n6.G)
    B=TextureSample(UVs=@n7) {samplerType=SAMPLERTYPE_LinearGrayscale, texture=&T_ElectricNoise}.G
   ))
  n9 = Power(Base=@n8) {constExponent=4}
  n10 = Multiply(A=@n8) {constB=0.5}
  n11 = TextureCoordinate()
  n12 = Power(Base=ComponentMask(None=Add(A=@n10, B=@n11)) {g=true})
  n13 = Clamp(None=Add(A=Clamp(None=@n9), B=Multiply(A=@n12)))
  n14 = VertexNormalWS()
  n15 = Subtract(A=WorldPosition().XYZ, B=ParticlePositionWS())
  n16 = OneMinus(None=VertexColor().R)
  Add(
   A=Multiply(
    A=Multiply(A=@n13, B=@n14)
    B=DynamicParameter() {defaultValue=<rgba 1 0 1 1>, paramNames [Param1, ScaleModifier, PannSpeed, DisplaceAmount]}.DisplaceAmount
   )
   B=Multiply(
    A=DynamicParameter() {defaultValue=<rgba 1 0 1 1>, paramNames [Param1, ScaleModifier, PannSpeed, DisplaceAmount]}.ScaleModifier
    B=Multiply(A=@n15, B=@n16)
   )
  )
  n17 = Fresnel() {exponent=1}
  n18 = ScalarParameter() {defaultValue=1.5, parameterName=AngleFadeAmount}
  n19 = Clamp(None=Power(
    Base=OneMinus(None=@n17)
    Exp=MaterialFunctionCall("FadeSensitivity (optional)"=@n18) {materialFunction=&FadeBasedOnViewAngle}.Result
   ))
  n20 = ComponentMask(None=MaterialFunctionCall(
    TextureObject=TextureObject() {texture=/Game/Effects/Fort_Effects/Effects/Characters/Generic/ResIn/DiffusedStylizedGrid}
    TextureSize=Constant3Vector() {constant=<rgba 200 200 200 0>}
   ) {materialFunction=&WorldAlignedTexture}.XY Texture) {r=true}
  n21 = MaterialFunctionCall() {materialFunction=&LinearGradient}
  n22 = Power(Base=@n21.VGradient) {constExponent=3}
  n23 = Clamp(None=Add(A=Multiply(A=@n20, B=@n22), B=Power(Base=@n21.VGradient) {constExponent=6}))
  DepthFade(Opacity=Multiply(A=ParticleColor().A, B=Multiply(A=@n19, B=@n23))) {fadeDistanceDefault=512}
  n24 = LinearInterpolate(
   A=VectorParameter() {parameterName=Color_1, defaultValue=<rgba 0.2396 0.7946 5 1>}.RGB
   B=VectorParameter() {parameterName=Color, defaultValue=<rgba 7.7272 0.2594 10 1>}.RGB
   Alpha=Add(A=@n20, B=@n22)
  )
  MaterialFunctionCall(Emissive=MaterialFunctionCall(LDR Input=Clamp(None=@n24), HDR Tint=@n24, Desaturate Luminance=ScalarParameter() {defaultValue=0.5, parameterName=HDRDesatLuminence}) {materialFunction=&DeriveHDRfromLDR}.RemappedRange) {materialFunction=&MF_FX_EmissiveAdjustments}
  n25 = TextureCoordinate()
  Multiply(A=Power(Base=ComponentMask(None=@n25) {g=true}), B=@n13)
  ObjectPositionWS()
  ParticleSize()
  ScalarParameter() {parameterName=ScaleAmount}
  TextureSample(UVs=Panner(Coordinate=TextureCoordinate() {uTiling=40, vTiling=5}, Time=@n3) {speedY=1}) {texture=&DiffusedStylizedGrid}
 ]
}
